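# Animated glitch material for the GlithGlithPSP vehicle wheel cosmetic, writing base color, emissive and normal through material attributes.
# Multiplied time sines jitter rim and tread UVs and switch between A/B tire masks and normals.
# Noise and scrolling datastream textures are sampled in screen space, corrected for aspect ratio, resolution and camera distance.
# Masks and Fresnel terms blend toward a two-color tint, with a Painted switch substituting paint color.
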Material /VehicleCosmetics/Wheels/GlithGlithPSP/Materials/M_Wheel_GlithGlithPSP
graph {
 properties {
  bAutomaticallySetUsageInEditor=false
  bUseMaterialAttributes=true
  bUsedWithMeshDeformer=true
  bUsedWithSkeletalMesh=true
  usedShadingModels=Default Lit
 }
 nodes [
  n1 = ViewSize()
  n2 = ComponentMask(None=@n1) {g=true}
  n3 = AppendVector(A=Constant() {r=1}, B=Divide(A=ComponentMask(None=@n1) {r=true}, B=@n2))
  MaterialFunctionCall!(Vector2=@n3, MaximumNumberOfDigits, UVs, DebugTextLocation RG_UpperRight BA_LowerLeft, Component Spacing) {materialFunction=/Engine/Functions/Engine_MaterialFunctions02/Utility/DebugFloat2Values}
  Multiply(A=@n1) {constB=0.001}
  n4 = ConstantBiasScale(None=NamedRerouteUsage()) {bias=0.2, scale=0.35}
  n5 = Multiply(A=@n4) {constB=0.15}
  n6 = Multiply(A=@n4) {constB=0.07}
  n7 = StaticSwitchParameter(True=Constant(), False=Max(A=Multiply(B=@n5), B=Clamp(None=@n6))) {parameterName=Painted}
  n8 = NamedRerouteUsage()
  n9 = ConstantBiasScale(None=NamedRerouteUsage()) {bias=0.5, scale=1}
  n10 = NamedRerouteUsage()
  n11 = ConstantBiasScale(None=@n10) {bias=0.15, scale=0.8}
  n12 = NamedRerouteUsage()
  n13 = NamedRerouteDeclaration(None=Time()) {name=Time, nodeColor=<rgba 0 1 0.9297 1>}
  n14 = Sine(None=Multiply(A=@n13) {constB=0.15}) {period=0.05}
  n15 = Sine(None=Multiply(A=@n13) {constB=0.66})
  n16 = Multiply(A=@n14, B=@n15)
  n17 = Multiply(A=@n15, B=Sine(None=Multiply(A=@n13) {constB=0.6}) {period=2})
  n18 = Multiply(A=@n16, B=@n17)
  n19 = LinearInterpolate(A=Constant2Vector(), B=Constant2Vector() {g=0.01, r=0.3}, Alpha=@n18)
  n20 = TextureCoordinate()
  n21 = ConstantBiasScale(None=NamedRerouteUsage()) {bias=-0.2, scale=1}
  n22 = TextureSampleParameter2D(UVs=LinearInterpolate(A=Add(A=@n19, B=@n20), B=@n20, Alpha=Clamp(None=@n21))) {parameterName=Rim_RGB, texture=/VehicleCosmetics/Wheels/GlithGlithPSP/Textures/Wheel_Glitch_RGB}
  n23 = Multiply(A=@n12, B=Add(A=@n22.R, B=@n22.R))
  n24 = TextureSampleParameter2D(UVs=@n20) {parameterName=Rim_RGB, texture=&Wheel_Glitch_RGB}
  n25 = Add(A=@n24.R, B=@n24.R)
  n26 = Constant2Vector()
  n27 = Constant2Vector() {g=0.03, r=0.08}
  n28 = Sine(None=Multiply(A=@n13) {constB=3})
  n29 = Multiply(A=@n18, B=Multiply(A=@n16, B=@n28))
  n30 = TextureCoordinate()
  n31 = LinearInterpolate(A=Add(A=LinearInterpolate(A=@n26, B=@n27, Alpha=@n29), B=@n30), B=@n30, Alpha=NamedRerouteUsage())
  n32 = TextureCoordinate()
  n33 = Clamp(None=ConstantBiasScale(None=NamedRerouteUsage()) {bias=-0.2, scale=1})
  n34 = Multiply(
   A=NamedRerouteUsage()
   B=TextureSampleParameter2D(UVs=LinearInterpolate(A=@n31, B=@n32, Alpha=@n33)) {parameterName=Tread_RGB_B, texture=/VehicleCosmetics/Wheels/GlithGlithPSP/Textures/Tire_GlitchB_RGB}.R
  )
  n35 = CameraVectorWS()
  n36 = OneMinus(None=ComponentMask(None=Transform(None=@n35) {transformSourceType=TRANSFORMSOURCE_World, transformType=TRANSFORM_Tangent}) {b=true})
  n37 = Power(Base=@n36) {constExponent=1.8}
  n38 = Power(Base=@n36) {constExponent=3}
  n39 = Multiply(A=LinearInterpolate(A=Clamp(None=@n37), B=Clamp(None=@n38), Alpha=NamedRerouteUsage()), B=Clamp(None=ConstantBiasScale(None=@n10)))
  n40 = LinearInterpolate(
   A=LinearInterpolate(B=@n11, Alpha=Max(A=@n23, B=@n25))
   B=Max(A=ConstantBiasScale(None=@n39) {bias=-0.1, scale=1}, B=LinearInterpolate(B=@n11, Alpha=@n34))
   Alpha=NamedRerouteUsage()
  )
  n41 = Constant() {r=2.5}
  n42 = NamedRerouteUsage()
  n43 = Multiply(A=Multiply(
    A=MaterialFunctionCall!(Normal, Exponent=@n41) {materialFunction=/VehicleCosmetics/Wheels/Cola_Holographic/Materials/MF_Fresnel_Classic}.Result
    B=ConstantBiasScale(None=@n42) {bias=2}
   ) {constB=0.25}) {constB=0.25}
  n44 = NamedRerouteUsage()
  n45 = Clamp(None=ConstantBiasScale(None=OneMinus(None=@n44)) {bias=0.1, scale=1})
  n46 = Clamp(None=NamedRerouteUsage())
  n47 = Max(A=NamedRerouteUsage(), B=NamedRerouteUsage())
  n48 = Clamp(None=ConstantBiasScale(None=Clamp(None=@n47)) {bias=0.15, scale=0.8})
  n49 = Multiply(A=@n11, B=Multiply(A=Multiply(A=@n46, B=@n48) {constB=5}) {constB=5})
  n50 = Max(A=Multiply(A=NamedRerouteUsage(), B=Clamp(None=@n9)), B=Add(A=Max(A=@n40, B=@n43), B=Multiply(A=@n45, B=@n49)))
  NamedRerouteDeclaration(None=LinearInterpolate(A=@n7, B=@n8, Alpha=@n50)) {name=BC, nodeColor=<rgba 0 0.1094 1 1>}
  NamedRerouteDeclaration(None=LinearInterpolate(A=@n7, B=@n8, Alpha=@n50)) {name=Emissive, nodeColor=<rgba 0.4297 1 0 1>}
  n51 = Distance(A=ActorPositionWS(), B=CameraPositionWS())
  n52 = ScreenPosition()
  n53 = Add(A=Multiply(A=Multiply(A=@n51) {constB=0.003}, B=Subtract(A=@n52.ViewportUV) {constB=0.5})) {constB=0.5}
  n54 = Subtract(A=@n2, B=Constant() {r=1250}) {constB=1250}
  n55 = Multiply(A=Divide(A=ComponentMask(None=@n53) {g=true, r=true}, B=@n3), B=LinearInterpolate(Alpha=Clamp(None=@n54)) {constA=10, constB=25})
  n56 = TextureSample(UVs=@n55) {texture=/VehicleCosmetics/Wheels/_SharedTextures/FX/Universal_Noise_39_RGB}
  NamedRerouteDeclaration(None=@n56.R) {name=Noise_R, nodeColor=<rgba 0 0.2734 1 1>}
  n57 = Divide(A=@n55, B=Constant2Vector() {g=1, r=1}) {constB=1}
  n58 = TextureSampleParameter2D(UVs=Panner(Coordinate=@n57) {speedX=-1}) {
   parameterName=GlitchMap
   samplerSource=SSM_Wrap_WorldGroupSettings
   texture=/VehicleCosmetics/Wheels/GlithGlithPSP/Textures/Tire_GlitchDatastream_RGB
  }
  NamedRerouteDeclaration(None=@n58.A) {name=Glitch_Fast_A, nodeColor=<rgba 0 0.2734 1 1>}
  n59 = VectorParameter() {parameterName=PaintColor, defaultValue=<rgba 0 1 0.6308 1>}
  n60 = Constant() {r=0.3}
  n61 = VectorParameter() {parameterName=Color1, defaultValue=<rgba 1 0.4179 0 1>}
  n62 = NamedRerouteUsage()
  n63 = NamedRerouteUsage()
  n64 = LinearInterpolate(A=Max(A=Multiply(A=@n12, B=@n22.G), B=@n24.G), B=Add(A=@n62, B=@n62), Alpha=@n63)
  n65 = VectorParameter() {parameterName=Color2, defaultValue=<rgba 0.9047 0.003 0.0027 1>}
  n66 = LinearInterpolate(A=Max(A=Multiply(A=@n22.R, B=@n12), B=@n24.R), B=@n34, Alpha=@n63)
  n67 = ScalarParameter() {defaultValue=2, parameterName=ColorMultiplier}
  n68 = Multiply(A=Max(A=Multiply(A=@n61.RGB, B=@n64), B=Multiply(A=@n65.RGB, B=@n66)), B=@n67)
  n69 = Desaturation(None=@n68) {luminanceFactors=<rgba 0.3 0.59 0.11 0>}
  n70 = LinearInterpolate(
   A=Multiply(A=Desaturation(None=@n61.RGB, Fraction=@n60) {luminanceFactors=<rgba 0.3 0.59 0.11 0>}) {constB=0.4475}
   B=@n68
   Alpha=Clamp(None=ConstantBiasScale(None=@n69) {bias=0, scale=5})
  )
  NamedRerouteDeclaration(None=Multiply(A=StaticSwitchParameter(True=@n59.RGB, False=@n70) {parameterName=Painted}, B=@n67)) {name=Color, nodeColor=<rgba 1 0.6563 0 1>}
  n71 = NamedRerouteUsage()
  n72 = Constant() {r=3}
  n73 = Floor(None=Fmod(A=Multiply(A=@n71) {constB=5}, B=@n72))
  n74 = TextureSampleParameter2D(UVs=Panner(Coordinate=@n57, Time=Divide(A=@n73, B=@n72)) {speedX=1}) {
   parameterName=GlitchMap
   samplerSource=SSM_Wrap_WorldGroupSettings
   texture=&Tire_GlitchDatastream_RGB
  }
  NamedRerouteDeclaration(None=@n74.R) {name=Glitch_R, nodeColor=<rgba 0 0.2734 1 1>}
  NamedRerouteDeclaration(None=@n74.G) {name=Glitch_G, nodeColor=<rgba 0 0.2734 1 1>}
  NamedRerouteDeclaration(None=@n74.A) {name=Glitch_A, nodeColor=<rgba 0 0.2734 1 1>}
  n75 = NamedRerouteUsage()
  n76 = NamedRerouteUsage()
  n77 = Ceil(None=@n17)
  n78 = Panner(Coordinate=TextureCoordinate() {vTiling=2}, Time=NamedRerouteUsage()) {speedX=0.2, speedY=0.1}
  n79 = NamedRerouteDeclaration(None=ConstantBiasScale(None=TextureSample(UVs=@n78) {texture=/VehicleCosmetics/Wheels/_SharedTextures/FX/TiledGradient_Balanced}.RGB) {bias=-0.5, scale=20}) {name=GlitchMaskGrad, nodeColor=<rgba 1 0 0.0703 1>}
  n80 = NamedRerouteDeclaration(None=Clamp(None=Multiply(A=@n77, B=@n79))) {name=Glitch Mask, nodeColor=<rgba 0 1 0.3672 1>}
  n81 = LinearInterpolate(A=@n75, B=@n76, Alpha=@n80)
  n82 = LinearInterpolate(A=@n76, B=@n75, Alpha=@n80)
  n83 = Ceil(None=@n15)
  n84 = Multiply(A=Multiply(A=Ceil(None=@n14), B=@n83), B=NamedRerouteUsage())
  n85 = Ceil(None=@n28)
  n86 = LinearInterpolate(
   A=LinearInterpolate(A=@n81, B=LinearInterpolate(A=@n81, B=@n82, Alpha=@n84), Alpha=@n83)
   B=LinearInterpolate(A=@n81, B=NamedRerouteUsage(), Alpha=Multiply(A=@n85, B=@n77))
   Alpha=NamedRerouteUsage()
  )
  n87 = NamedRerouteUsage()
  NamedRerouteDeclaration(None=ComponentMask(None=LinearInterpolate(B=@n86, Alpha=@n87)) {g=true}) {name=TextGlitched, nodeColor=<rgba 0 0.3906 1 1>}
  n88 = TextureSampleParameter2D() {
   parameterName=Normal
   samplerSource=SSM_Wrap_WorldGroupSettings
   samplerType=SAMPLERTYPE_Normal
   texture=/VehicleCosmetics/Wheels/GlithGlithPSP/Textures/Tire_Glitch_N
  }
  n89 = TextureSampleParameter2D() {
   parameterName=Normal_B
   samplerSource=SSM_Wrap_WorldGroupSettings
   samplerType=SAMPLERTYPE_Normal
   texture=/VehicleCosmetics/Wheels/GlithGlithPSP/Textures/Tire_GlitchB_N
  }
  n90 = LinearInterpolate(A=@n88.RGB, B=@n89.RGB, Alpha=@n83)
  n91 = LinearInterpolate(A=@n88.RGB, B=@n89.RGB, Alpha=@n85)
  n92 = NamedRerouteUsage()
  NamedRerouteDeclaration(None=LinearInterpolate(A=Constant3Vector() {constant=<rgba 0 0 1 0>}, B=LinearInterpolate(A=@n90, B=@n91, Alpha=@n92), Alpha=NamedRerouteUsage())) {name=Normal, nodeColor=<rgba 1 0 0.75 1>}
  n93 = TextureSampleParameter2D() {parameterName=Mask_A, samplerSource=SSM_Wrap_WorldGroupSettings, texture=/VehicleCosmetics/Wheels/GlithGlithPSP/Textures/Tire_Glitch_RGB}
  NamedRerouteDeclaration(None=@n93.RGB) {name=Mask_A_RGB, nodeColor=<rgba 0 0.6719 1 1>}
  n94 = TextureSampleParameter2D() {parameterName=Mask_B, samplerSource=SSM_Wrap_WorldGroupSettings, texture=&Tire_GlitchB_RGB}
  NamedRerouteDeclaration(None=@n94.RGB) {name=Mask_B_RGB, nodeColor=<rgba 0 0.6719 1 1>}
  NamedRerouteDeclaration(None=@n94.B) {name=Mask_B_B, nodeColor=<rgba 0 0.6719 1 1>}
  NamedRerouteDeclaration(None=VertexColor().R) {name=VC_R, nodeColor=<rgba 1 0 0 1>}
  NamedRerouteDeclaration(None=@n56.A) {name=Noise_A, nodeColor=<rgba 0 0.2734 1 1>}
  Normalize(VectorInput=@n52.ViewportUV)
  SetMaterialAttributes!(MaterialAttributes, Base Color=NamedRerouteUsage(), Emissive Color=NamedRerouteUsage(), Normal=NamedRerouteUsage()) {attributeSetTypes [#1, #2, #3]}
 ]
}
